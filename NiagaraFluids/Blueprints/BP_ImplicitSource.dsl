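; Niagara Fluids implicit source actor that exposes a directional velocity and its own motion velocity, presumably for a fluid simulation to read.
; Construction script derives Velocity from the actor's forward direction and VelocityMagnitude, and stretches StaticMesh1 along X to match that magnitude.
; Each tick derives ActorVel from the position change over delta time, scaled by InheritVelocity, using PrevPos seeded at BeginPlay.
; The overlap event is empty.
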(event EventBeginPlay
  (bind _returnvalue_location (Math|Transform|BreakTransform (Transformation|GetActorTransform)))
  (Variables|Default|SetPrevPos _returnvalue_location))

(event Collision|EventActorBeginOverlap
  (params (OtherActor :type Actor%20Object%20Reference)))

(event EventTick
  (params (DeltaSeconds :type Float%20%28single-precision%29))
  (bind _returnvalue_location (Math|Transform|BreakTransform (Transformation|GetActorTransform)))
  (bind _output_get (Variables|Default|SetActorVel (* (/ (- _returnvalue_location (Variables|Default|GetPrevPos)) DeltaSeconds) (Variables|Default|GetInheritVelocity))))
  (Variables|Default|SetPrevPos _returnvalue_location))

(fn ConstructionScript
  (bind _velocitymagnitude (Variables|Default|GetVelocityMagnitude))
  (Variables|Default|SetVelocity (* (Math|Vector|RotateVector "1.000000, 0, 0" (Transformation|GetActorRotation)) (* _velocitymagnitude 1e+02)))
  (Transformation|SetWorldScale3D (Variables|Default|GetStaticMesh1) (Math|Vector|MakeVector (+ _velocitymagnitude 0.001) 1.0 1.0)))
