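; BP_WaterRenderer displays a Niagara shallow water simulation on a plane using a dynamic instance of a surface or debug material.
; BeginPlay creates float render targets for simulation, normals and optional foam and profile buffer, binding each to both the Niagara system's User parameters and the material's textures.
; The first Tick activates the Niagara component, guarded by a flag.
; WriteData only fetches and casts the grid collection data interface, writing nothing.
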(event EventBeginPlay
  (bind _render_mtl_mid (Variables|Default|GetRender_mtl_mid))
  (bind _system (Variables|Default|GetSystem))
  (bind _niagaracomponent (Variables|/Script/Niagara.NiagaraActor|GetNiagaraComponent _system))
  (Utilities|FlowControl|Sequence
    (:then_0
      (if (Variables|Default|Getdebug)
        (bind _returnvalue (Rendering|Material|CreateDynamicMaterialInstance "/NiagaraFluids/Materials/ShallowWater/DebugWaterMaterial.DebugWaterMaterial" "None" "None" :!class /Script/Engine.KismetMaterialLibrary))
        (bind _output_get (Variables|Default|SetRender_mtl_mid _returnvalue))
        (Rendering|Material|SetMaterial (Variables|Default|GetPlane) 0 _output_get)
        (else
          (bind _returnvalue_2 (Rendering|Material|CreateDynamicMaterialInstance (Variables|Default|GetSurfaceMaterial) "None" "None" :!class /Script/Engine.KismetMaterialLibrary))
          (bind _output_get_2 (Variables|Default|SetRender_mtl_mid _returnvalue_2))
          (Rendering|Material|SetMaterial (Variables|Default|GetPlane) 0 _output_get_2))))
    (:then_1
      (bind _returnvalue_3 (Rendering|CreateRenderTarget2D 1 1 "RTF_RGBA32f"))
      (bind _output_get_3 (Variables|Default|SetSimTexture_RT _returnvalue_3))
      (Niagara|SetNiagaraVariable(TextureRenderTarget) _niagaracomponent "User.SimRT" _output_get_3)
      (Rendering|Material|SetTextureParameterValue _render_mtl_mid "Tex" _output_get_3))
    (:then_2
      (if (Variables|Default|GetUseFoam)
        (bind _returnvalue_4 (Rendering|CreateRenderTarget2D 1 1 "RTF_R32f"))
        (bind _output_get_4 (Variables|Default|SetFoamTexture_RT _returnvalue_4))
        (Niagara|SetNiagaraVariable(Object) _niagaracomponent "User.FoamRT" _output_get_4)
        (Rendering|Material|SetTextureParameterValue _render_mtl_mid "FoamTex" _output_get_4)))
    (:then_3
      (if (Variables|Default|GetUseProfileBuffer)
        (bind _returnvalue_5 (Rendering|CreateRenderTarget2D 1 1 "RTF_R32f"))
        (bind _output_get_5 (Variables|Default|SetProfileBuffer_RT _returnvalue_5))
        (Niagara|SetNiagaraVariable(Object) _niagaracomponent "User.ProfileBufferRT" _output_get_5)
        (Rendering|Material|SetTextureParameterValue _render_mtl_mid "ProfileBufferTex" _output_get_5)))
    (:then_4
      (bind _returnvalue_6 (Rendering|CreateRenderTarget2D 1 1 "RTF_RGBA32f"))
      (bind _output_get_6 (Variables|Default|SetNormalTexture_RT _returnvalue_6))
      (Niagara|SetNiagaraVariable(TextureRenderTarget) _niagaracomponent "User.NormalRT" _output_get_6)
      (Rendering|Material|SetTextureParameterValue _render_mtl_mid "NormalTex" _output_get_6))))

(event EventTick
  (params (DeltaSeconds :type Float%20%28single-precision%29))
  (if (not (Variables|Default|GetAc))
    (Components|Activation|Activate (Variables|/Script/Niagara.NiagaraActor|GetNiagaraComponent (Variables|Default|GetSystem)) false)
    (Variables|Default|SetAc true)))

(event Custom|WriteData
  (bind _returnvalue (Niagara|GetDataInterface (Variables|/Script/Niagara.NiagaraActor|GetNiagaraComponent (Variables|Default|GetSystem)) "Emitter.GridCollection"))
  (if (Utilities|IsValid _returnvalue :!class /Script/Engine.KismetSystemLibrary)
    (Utilities|Casting|CastToNiagaraDataInterfaceGrid2DCollection _returnvalue)))

(fn ConstructionScript)
