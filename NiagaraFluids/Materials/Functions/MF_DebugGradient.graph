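# MF_DebugGradient is a Niagara Fluids material function, exposed to the library, that maps a scalar X to a blue-green-red heatmap color for debugging.
# One Custom node closes its generated HLSL function early to declare global R2G, G2B and heatmap helpers, then opens a dummy function to balance the braces.
# A second Custom node takes that node as an unused input, making the helpers available, and outputs heatmap(X).
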MaterialFunction /NiagaraFluids/Materials/Functions/MF_DebugGradient
graph {
 properties {bExposeToLibrary=true}
 nodes [
  MaterialExpressionCustom_0 = Custom!(None) {code="return 0;\r\n}\r\n\r\nfloat3 R2G(float x)\r\n{\r\n    if(x < 0.5)\r\n    {\r\n        return float3(1.f, 2.f * x, 0.0); \r\n    }\r\n    else\r\n    {\r\n        return float3(2.f - 2.f * x, 1.f, 0.0);\r\n    }\r\n}\r\n\r\nfloat3 G2B(float x)\r\n{\r\n    if(x < 0.5)\r\n    {\r\n        return float3(0.0, 1.f, 2.f * x); \r\n    }\r\n    else\r\n    {\r\n        return float3(0.0, 2.f - 2.f * x, 1.f);\r\n    }  \r\n}\r\nfloat3 heatmap(float x)\r\n{\r\n    float InvertedX = x;\r\n    if(InvertedX < 0.5)\r\n    {\r\n        float RemappedFirstHalf = 1.f - 2.f * InvertedX;\r\n        return G2B(RemappedFirstHalf);\r\n    }\r\n    else\r\n    {\r\n     \tfloat RemappedSecondHalf = 2.f - 2.f * InvertedX; \r\n        return R2G(RemappedSecondHalf);\r\n    }\r\n}\r\n\r\nfloat3 asdf()\r\n{\r\nreturn float3(0,0,0);"}
  MaterialExpressionFunctionInput_0 = FunctionInput() {inputName=X, inputType=FunctionInput_Scalar}
  n1 = Custom!(A=@MaterialExpressionCustom_0, X=@MaterialExpressionFunctionInput_0) {
   code="return float3(heatmap(X));"
   inputs [
    {inputName=A, input {expression=&self:MaterialExpressionCustom_0}}
    {inputName=X, input {expression=&self:MaterialExpressionFunctionInput_0}}
   ]
  }
  FunctionOutput(None=@n1)
 ]
}
